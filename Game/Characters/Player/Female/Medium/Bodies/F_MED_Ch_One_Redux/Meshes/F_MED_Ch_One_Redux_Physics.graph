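# Physics asset for the F_MED_Ch_One_Redux female player body, giving jointed physics to dangling accessories: dog-tag necklace, armband, laces and shin guard straps.
# Capsule chains with limited swing and twist hang from spine_05, the right upper arm, right calf and left calf, branching at dog tags, laces and armband.
# Necklace and lace master links, plus the arm-to-spine joint, swing and twist freely.
# Larger capsules on spine_05 and the right calf approximate body volume.
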PhysicsAsset /Game/Characters/Player/Female/Medium/Bodies/F_MED_Ch_One_Redux/Meshes/F_MED_Ch_One_Redux_Physics
bodies {
 DYN_Armband [{shapeType=Capsule, length=2.25, radius=0.75, center=<1.6685 0 0>, rotation=<rot 90 0.0395 0.0394>}]
 DYN_ArmbandL1 [{shapeType=Capsule, length=2.25, radius=0.75, center=<1.6685 0 0>, rotation=<rot 90 0.0099 0.0099>}]
 DYN_ArmbandL2 [{shapeType=Capsule, length=2.25, radius=0.75, center=<1.6685 0 0>, rotation=<rot 90 179.9605 179.9604>}]
 DYN_ArmbandR1 [{shapeType=Capsule, length=2.25, radius=0.75, center=<1.6685 0 0>, rotation=<rot 90 -2.5305 -2.5305>}]
 DYN_ArmbandR2 [{shapeType=Capsule, length=2.25, radius=0.75, center=<1.6685 0 0>, rotation=<rot 90 0.0395 0.0394>}]
 DYN_ArmbandR3 [{shapeType=Capsule, length=2.25, radius=0.75, center=<1.6685 0 0>, rotation=<rot 90 0.0198 0.0197>}]
 DYN_Laces [{shapeType=Capsule, length=3.75, radius=1.75, center=<2.7786 0 0>}]
 DYN_LacesL1 [{shapeType=Capsule, length=1.75, radius=1.25, center=<1.8347 0 0>}]
 DYN_LacesL2 [{shapeType=Capsule, length=1.25, radius=1.5, center=<2.0812 0 0>, rotation=<rot 0.0004 0 0>}]
 DYN_LacesR1 [{shapeType=Capsule, length=2, radius=1.25, center=<1.418 0 0>}]
 DYN_LacesR2 [{shapeType=Capsule, length=1.75, radius=1.5, center=<2.5282 0 0>, rotation=<rot -90 0 -89.9996>}]
 DYN_Laces_Master [{shapeType=Capsule, length=0.5, radius=0.5}]
 DYN_ShinGuardStrapLower_1 [{shapeType=Capsule, length=0.5, radius=1, center=<1.302 0 0>, rotation=<rot 90 -0.0197 -0.0196>}]
 DYN_ShinGuardStrapLower_2 [{shapeType=Capsule, length=0.5, radius=1, center=<1.302 0 0>, rotation=<rot 90 -0.0033 -0.0032>}]
 DYN_ShinGuardStrapUpper_1 [{shapeType=Capsule, length=1.5, radius=1, center=<1.8433 0 0>, rotation=<rot 90 -0.0396 -0.0396>}]
 DYN_ShinGuardStrapUpper_2 [{shapeType=Capsule, length=0.5, radius=1, center=<1.302 0 0>, rotation=<rot 90 0.0198 0.0198>}]
 Dyn_Necklace [{shapeType=Capsule, length=0.5, radius=0.5}]
 Dyn_Necklace_Back_01 [{shapeType=Capsule, length=3.75, radius=1.25, center=<1.6765 0 0>}]
 Dyn_Necklace_Cntr_01 [{shapeType=Capsule, length=9.1212, radius=0.7572, center=<1.8794 -0.1759 0>}]
 Dyn_Necklace_Cntr_02 [{shapeType=Capsule, length=9.1067, radius=0.7717, center=<1.4459 -0.2185 0>}]
 Dyn_Necklace_Cntr_03 [{shapeType=Capsule, length=2, radius=1.5, center=<2.5357 0.0001 0>}]
 Dyn_Necklace_Dogtag_L [{shapeType=Capsule, length=1.75, radius=1, center=<0.8678 0.0001 -0.0001>, rotation=<rot -55.0004 90.0098 -90.008>}]
 Dyn_Necklace_Dogtag_R [{shapeType=Capsule, length=0, radius=1.25, center=<1.0529 -0.1949 0.5361>, rotation=<rot 0 0 -20.0002>}]
 Dyn_Necklace_MStR [{shapeType=Capsule, length=0.5, radius=0.5}]
 calf_twist_02_l [{shapeType=Capsule, length=0.5, radius=0.5}]
 calf_twist_02_r [{shapeType=Capsule, length=13.6778, radius=4.0091, center=<8.2476 1.5004 0.3489>, rotation=<rot -2.6414 -84.3653 -88.2148>}]
 spine_05 [
  {shapeType=Capsule, length=15.7694, radius=8.6689}
  {shapeType=Capsule, length=10.5893, radius=8.0357, center=<-1.9421 2.6898 0.0004>, rotation=<rot 2.1722 50.6008 178.2178>}
  {shapeType=Capsule, length=10.5893, radius=8.0357, center=<-7.5729 6.1407 0.1695>, rotation=<rot 2.1722 50.6018 178.2178>}
 ]
 upperarm_twist_01_r [{shapeType=Capsule, length=0.5, radius=0.5}]
}
constraints [
 {bone1Name=Dyn_Necklace_Back_01, bone2Name=Dyn_Necklace, swing1LimitDegrees=20, swing2LimitDegrees=2, twistLimitDegrees=5}
 {bone1Name=Dyn_Necklace_Cntr_01, bone2Name=Dyn_Necklace, swing1LimitDegrees=30, swing2LimitDegrees=10, twistLimitDegrees=7}
 {bone1Name=Dyn_Necklace_Cntr_02, bone2Name=Dyn_Necklace_Cntr_01, swing1LimitDegrees=30, swing2LimitDegrees=10, twistLimitDegrees=7}
 {bone1Name=Dyn_Necklace_Cntr_03, bone2Name=Dyn_Necklace_Cntr_02, swing1LimitDegrees=30, swing2LimitDegrees=10, twistLimitDegrees=7}
 {bone1Name=Dyn_Necklace_Dogtag_R, bone2Name=Dyn_Necklace_Cntr_03, swing1LimitDegrees=30, swing2LimitDegrees=30, twistLimitDegrees=15}
 {bone1Name=Dyn_Necklace_Dogtag_L, bone2Name=Dyn_Necklace_Cntr_03, swing1LimitDegrees=30, swing2LimitDegrees=30, twistLimitDegrees=15}
 {
  bone1Name=Dyn_Necklace
  bone2Name=Dyn_Necklace_MStR
  swing1LimitDegrees=45
  swing1Motion=Free
  swing2LimitDegrees=45
  swing2Motion=Free
  twistLimitDegrees=45
  twistMotion=Free
 }
 {
  bone1Name=Dyn_Necklace_MStR
  bone2Name=spine_05
  swing1LimitDegrees=45
  swing1Motion=Free
  swing2LimitDegrees=45
  swing2Motion=Free
  twistLimitDegrees=45
  twistMotion=Free
 }
 {bone1Name=DYN_ShinGuardStrapUpper_2, bone2Name=DYN_ShinGuardStrapUpper_1, swing1LimitDegrees=5, swing2LimitDegrees=17, twistLimitDegrees=5}
 {bone1Name=DYN_ShinGuardStrapLower_2, bone2Name=DYN_ShinGuardStrapLower_1, swing1LimitDegrees=5, swing2LimitDegrees=17, twistLimitDegrees=5}
 {bone1Name=DYN_Laces, bone2Name=DYN_Laces_Master, swing1LimitDegrees=45, swing2LimitDegrees=45, twistLimitDegrees=5}
 {bone1Name=DYN_LacesR1, bone2Name=DYN_Laces, swing1LimitDegrees=45, swing2LimitDegrees=45, twistLimitDegrees=5}
 {bone1Name=DYN_LacesR2, bone2Name=DYN_LacesR1, swing1LimitDegrees=45, swing2LimitDegrees=45, twistLimitDegrees=5}
 {bone1Name=DYN_LacesL1, bone2Name=DYN_Laces, swing1LimitDegrees=45, swing2LimitDegrees=45, twistLimitDegrees=5}
 {bone1Name=DYN_LacesL2, bone2Name=DYN_LacesL1, swing1LimitDegrees=45, swing2LimitDegrees=45, twistLimitDegrees=5}
 {
  bone1Name=DYN_Laces_Master
  bone2Name=calf_twist_02_r
  swing1LimitDegrees=45
  swing1Motion=Free
  swing2LimitDegrees=45
  swing2Motion=Free
  twistLimitDegrees=45
  twistMotion=Free
 }
 {bone1Name=DYN_ShinGuardStrapUpper_1, bone2Name=calf_twist_02_l, swing1LimitDegrees=5, swing2LimitDegrees=17, twistLimitDegrees=5}
 {bone1Name=DYN_ShinGuardStrapLower_1, bone2Name=calf_twist_02_l, swing1LimitDegrees=5, swing2LimitDegrees=17, twistLimitDegrees=5}
 {
  bone1Name=upperarm_twist_01_r
  bone2Name=spine_05
  swing1LimitDegrees=45
  swing1Motion=Free
  swing2LimitDegrees=45
  swing2Motion=Free
  twistLimitDegrees=45
  twistMotion=Free
 }
 {bone1Name=DYN_Armband, bone2Name=upperarm_twist_01_r, swing1LimitDegrees=10, swing2LimitDegrees=10, twistLimitDegrees=3}
 {bone1Name=DYN_ArmbandR1, bone2Name=DYN_Armband, swing1LimitDegrees=12, swing2LimitDegrees=12, twistLimitDegrees=5}
 {bone1Name=DYN_ArmbandR2, bone2Name=DYN_ArmbandR1, swing1LimitDegrees=12, swing2LimitDegrees=12, twistLimitDegrees=5}
 {bone1Name=DYN_ArmbandR3, bone2Name=DYN_ArmbandR2, swing1LimitDegrees=12, swing2LimitDegrees=12, twistLimitDegrees=5}
 {bone1Name=DYN_ArmbandL1, bone2Name=DYN_Armband, swing1LimitDegrees=12, swing2LimitDegrees=12, twistLimitDegrees=5}
 {bone1Name=DYN_ArmbandL2, bone2Name=DYN_ArmbandL1, swing1LimitDegrees=12, swing2LimitDegrees=12, twistLimitDegrees=5}
]
