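# Post-process style AnimBlueprint for the F_MED_Ch_One_Redux player body that copies the attached parent's pose and adds secondary motion to accessories.
# Three AnimDynamics and RigidBody branches blend on crouching and targeting, varying simulation base bone and armband constraints.
# State-driven bone offsets, scales and constraints then correct kneepads, laces, armband and necklace for crouch, air, emote and front-end poses.
# The event graph sets isRBANdisabled, which the AnimGraph never reads.
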AnimBlueprint /Game/Characters/Player/Female/Medium/Bodies/F_MED_Ch_One_Redux/Meshes/F_MED_Ch_One_Redux_AnimBP
target_skeleton=&/F_MED_Ch_One_Redux_Skeleton
graphs [
 AnimGraph {nodes [
   n1 = Get bIsCrouching()
   n2 = "Transform (Modify) Bone - Bone: DYN_Kneepad"!(
    ComponentPose=Local To Component(LocalPose=Copy Pose From Mesh() {bUseAttachedParent=true})
    bAlphaBoolEnabled:bool=@n1
    Translation="-2.5,-1,0"
    Rotation="0,-36,1"
    Scale="1,1,1"
   ) -> (Pose) {
    alphaInputType=Bool
    rotationMode=BMM_Additive
    rotationSpace=BCS_ParentBoneSpace
    translationMode=BMM_Additive
    translationSpace=BCS_ParentBoneSpace
    alphaBoolBlend {blendOption=Linear}
    boneToModify {boneName=DYN_Kneepad}
    rotation {roll=1, yaw=-36}
    translation=<-2.5 -1>
   }
   n3 = "Transform (Modify) Bone - Bone: dyn_glute_l"!(ComponentPose=@n2, bAlphaBoolEnabled:bool=@n1, Translation="-3,0,1", Rotation="0,0,0", Scale="1,1,1") -> (Pose) {
    alphaInputType=Bool
    rotationMode=BMM_Additive
    rotationSpace=BCS_ParentBoneSpace
    translationMode=BMM_Additive
    translationSpace=BCS_ParentBoneSpace
    alphaBoolBlend {blendOption=Linear}
    boneToModify {boneName=dyn_glute_l}
    translation {x=-3, z=1}
   }
   n4 = Anim Dynamics - Bone: Dyn_Necklace_MStR(ComponentPose=@n3, Alpha=0.3, ExternalForce="0,0,0") {
    angularDampingOverride=0.9
    bGravityOverrideInSimSpace=true
    bOverrideAngularDamping=true
    bOverrideLinearDamping=true
    bUseGravityOverride=true
    linearDampingOverride=0.99
    simulationSpace=BoneRelative
    alphaBoolBlend {blendOption=Linear}
    boundBone {boneName=Dyn_Necklace_MStR}
    componentLinearAccScale=<0.1 0.1 0.1>
    componentLinearVelScale=<0.6 0.6 0.6>
    gravityOverride {z=-250}
    physicsBodyDefinitions [{
      sphereCollisionRadius=0
      boundBone {boneName=Dyn_Necklace_MStR}
      boxExtents=<10 10 10>
      constraintSetup {angularLimitsMax=<0 0 0>, angularLimitsMin=<0 0 0>, angularTarget=<0 0 0>, linearAxesMax=<0 0 2>, linearAxesMin=<0 0 -0.5>}
      localJointOffset=<0 0 -3>
     }]
    relativeSpaceBone {boneName=root}
   }
   n5 = RigidBody(ComponentPose=@n4) {
    bOverrideWorldGravity=true
    overridePhysicsAsset=&/F_MED_Ch_One_Redux_Physics
    simulationSpace=BaseBoneSpace
    alphaBoolBlend {blendOption=Linear}
    baseBoneRef {boneName=root}
    componentLinearAccScale=<0.3 0.3 0.3>
    componentLinearVelScale=<0.6 0.6 0.6>
    overrideWorldGravity {z=-980}
   }
   n6 = "Transform (Modify) Bone - Bone: Dyn_Necklace_MStR"(ComponentPose=@n5, Alpha=1, Translation="0.000,0.000,-0.3", Rotation="0,0,0", Scale="1,1,1") {
    translationMode=BMM_Additive
    translationSpace=BCS_BoneSpace
    alphaBoolBlend {blendOption=Linear}
    boneToModify {boneName=Dyn_Necklace_MStR}
    translation {z=-0.3}
   }
   n7 = Get bIsCrouching()
   n8 = "Transform (Modify) Bone - Bone: DYN_Kneepad"(
    ComponentPose=Local To Component(LocalPose=Copy Pose From Mesh() {bUseAttachedParent=true})
    bAlphaBoolEnabled=@n7
    Translation="-2.5,-1,0"
    Rotation="0,-36,1"
    Scale="1,1,1"
   ) {
    alphaInputType=Bool
    rotationMode=BMM_Additive
    rotationSpace=BCS_ParentBoneSpace
    translationMode=BMM_Additive
    translationSpace=BCS_ParentBoneSpace
    alphaBoolBlend {blendOption=Linear}
    boneToModify {boneName=DYN_Kneepad}
    rotation {roll=1, yaw=-36}
    translation=<-2.5 -1>
   }
   n9 = "Transform (Modify) Bone - Bone: dyn_glute_l"(ComponentPose=@n8, bAlphaBoolEnabled=@n7, Translation="-3,0,1", Rotation="0,0,0", Scale="1,1,1") {
    alphaInputType=Bool
    rotationMode=BMM_Additive
    rotationSpace=BCS_ParentBoneSpace
    translationMode=BMM_Additive
    translationSpace=BCS_ParentBoneSpace
    alphaBoolBlend {blendOption=Linear}
    boneToModify {boneName=dyn_glute_l}
    translation {x=-3, z=1}
   }
   n10 = Anim Dynamics - Bone: Dyn_Necklace_MStR(ComponentPose=@n9, Alpha=0.3, ExternalForce="0,0,0") {
    angularDampingOverride=0.9
    bGravityOverrideInSimSpace=true
    bOverrideAngularDamping=true
    bOverrideLinearDamping=true
    bUseGravityOverride=true
    linearDampingOverride=0.99
    simulationSpace=BoneRelative
    alphaBoolBlend {blendOption=Linear}
    boundBone {boneName=Dyn_Necklace_MStR}
    componentLinearAccScale=<0.1 0.1 0.1>
    componentLinearVelScale=<0.6 0.6 0.6>
    gravityOverride {z=-250}
    physicsBodyDefinitions [{
      sphereCollisionRadius=0
      boundBone {boneName=Dyn_Necklace_MStR}
      boxExtents=<10 10 10>
      constraintSetup {angularLimitsMax=<0 0 0>, angularLimitsMin=<0 0 0>, angularTarget=<0 0 0>, linearAxesMax=<0 0 2>, linearAxesMin=<0 0 -0.5>}
      localJointOffset=<0 0 -3>
     }]
    relativeSpaceBone {boneName=root}
   }
   n11 = RigidBody(ComponentPose=@n10) {
    bOverrideWorldGravity=true
    overridePhysicsAsset=&F_MED_Ch_One_Redux_Physics
    simulationSpace=BaseBoneSpace
    alphaBoolBlend {blendOption=Linear}
    baseBoneRef {boneName=root}
    componentLinearAccScale=<0.3 0.3 0.3>
    componentLinearVelScale=<0.6 0.6 0.6>
    overrideWorldGravity {z=-980}
   }
   n12 = "Constraint - DYN_Armband (P)"(ComponentPose=@n11, Alpha=1, ConstraintWeights_0=1) {
    alphaBoolBlend {blendOption=Linear}
    alphaScaleBiasClamp {scale=0.23}
    boneToModify {boneName=DYN_Armband}
    constraintSetup [{targetBone {boneName=upperarm_twist_02_r}}]
    constraintWeights [1]
   }
   n13 = "Constraint - DYN_ArmbandR1 (P)"(ComponentPose=@n12, Alpha=1, ConstraintWeights_0=1) {
    alphaBoolBlend {blendOption=Linear}
    alphaScaleBiasClamp {scale=0.23}
    boneToModify {boneName=DYN_ArmbandR1}
    constraintSetup [{targetBone {boneName=upperarm_twist_02_r}}]
    constraintWeights [1]
   }
   n14 = "Constraint - DYN_ArmbandL1 (P)"(ComponentPose=@n13, Alpha=1, ConstraintWeights_0=1) {
    alphaBoolBlend {blendOption=Linear}
    alphaScaleBiasClamp {scale=0.23}
    boneToModify {boneName=DYN_ArmbandL1}
    constraintSetup [{targetBone {boneName=upperarm_twist_02_r}}]
    constraintWeights [1]
   }
   n15 = "Constraint - DYN_ArmbandR2 (P)"(ComponentPose=@n14, Alpha=1, ConstraintWeights_0=1) {
    alphaBoolBlend {blendOption=Linear}
    alphaScaleBiasClamp {scale=0.23}
    boneToModify {boneName=DYN_ArmbandR2}
    constraintSetup [{targetBone {boneName=upperarm_twist_02_r}}]
    constraintWeights [1]
   }
   n16 = "Constraint - DYN_ArmbandL2 (P)"(ComponentPose=@n15, Alpha=1, ConstraintWeights_0=1) {
    alphaBoolBlend {blendOption=Linear}
    alphaScaleBiasClamp {scale=0.23}
    boneToModify {boneName=DYN_ArmbandL2}
    constraintSetup [{targetBone {boneName=upperarm_twist_02_r}}]
    constraintWeights [1]
   }
   n17 = "Constraint - DYN_ArmbandR3 (P)"(ComponentPose=@n16, Alpha=1, ConstraintWeights_0=1) {
    alphaBoolBlend {blendOption=Linear}
    alphaScaleBiasClamp {scale=0.23}
    boneToModify {boneName=DYN_ArmbandR3}
    constraintSetup [{targetBone {boneName=upperarm_twist_02_r}}]
    constraintWeights [1]
   }
   n18 = "Transform (Modify) Bone - Bone: Dyn_Necklace_MStR"(ComponentPose=@n17, Alpha=1, Translation="0.000,0.000,-0.3", Rotation="0,0,0", Scale="1,1,1") {
    translationMode=BMM_Additive
    translationSpace=BCS_BoneSpace
    alphaBoolBlend {blendOption=Linear}
    boneToModify {boneName=Dyn_Necklace_MStR}
    translation {z=-0.3}
   }
   n19 = Blend Poses by bool(
    BlendPose_0=Component To Local(ComponentPose=@n18)
    BlendPose_1=Component To Local(ComponentPose=@n6)
    BlendTime_0=0.2
    BlendTime_1=0.2
    bActiveValue=Get bIsTargeting()
   ) {blendType=Linear}
   n20 = Get bIsCrouching()
   n21 = "Transform (Modify) Bone - Bone: DYN_Kneepad"(
    ComponentPose=Local To Component(LocalPose=Copy Pose From Mesh() {bUseAttachedParent=true})
    bAlphaBoolEnabled=@n20
    Translation="-2.5,-1,0"
    Rotation="0,-36,1"
    Scale="1,1,1"
   ) {
    alphaInputType=Bool
    rotationMode=BMM_Additive
    rotationSpace=BCS_ParentBoneSpace
    translationMode=BMM_Additive
    translationSpace=BCS_ParentBoneSpace
    alphaBoolBlend {blendOption=Linear}
    boneToModify {boneName=DYN_Kneepad}
    rotation {roll=1, yaw=-36}
    translation=<-2.5 -1>
   }
   n22 = "Transform (Modify) Bone - Bone: dyn_glute_l"(ComponentPose=@n21, bAlphaBoolEnabled=@n20, Translation="-3,0,1", Rotation="0,0,0", Scale="1,1,1") {
    alphaInputType=Bool
    rotationMode=BMM_Additive
    rotationSpace=BCS_ParentBoneSpace
    translationMode=BMM_Additive
    translationSpace=BCS_ParentBoneSpace
    alphaBoolBlend {blendOption=Linear}
    boneToModify {boneName=dyn_glute_l}
    translation {x=-3, z=1}
   }
   n23 = Anim Dynamics - Bone: Dyn_Necklace_MStR(ComponentPose=@n22, Alpha=0.3, ExternalForce="0,0,0") {
    angularDampingOverride=0.9
    bGravityOverrideInSimSpace=true
    bOverrideAngularDamping=true
    bOverrideLinearDamping=true
    bUseGravityOverride=true
    linearDampingOverride=0.99
    simulationSpace=BoneRelative
    alphaBoolBlend {blendOption=Linear}
    boundBone {boneName=Dyn_Necklace_MStR}
    componentLinearAccScale=<0.1 0.1 0.1>
    componentLinearVelScale=<0.6 0.6 0.6>
    gravityOverride {z=-250}
    physicsBodyDefinitions [{
      sphereCollisionRadius=0
      boundBone {boneName=Dyn_Necklace_MStR}
      boxExtents=<10 10 10>
      constraintSetup {angularLimitsMax=<0 0 0>, angularLimitsMin=<0 0 0>, angularTarget=<0 0 0>, linearAxesMax=<0 0 2>, linearAxesMin=<0 0 -0.5>}
      localJointOffset=<0 0 -3>
     }]
    relativeSpaceBone {boneName=root}
   }
   n24 = RigidBody(ComponentPose=@n23) {
    bOverrideWorldGravity=true
    overridePhysicsAsset=&F_MED_Ch_One_Redux_Physics
    simulationSpace=BaseBoneSpace
    alphaBoolBlend {blendOption=Linear}
    alphaScaleBiasClamp {scale=0.5}
    baseBoneRef {boneName=spine_05}
    componentAppliedLinearAccClamp=<5000 5000 5000>
    componentLinearAccScale=<0.5 0.5 0.5>
    overrideWorldGravity {z=-980}
   }
   n25 = Component To Local(ComponentPose=@n24)
   n26 = Get bIsCrouching()
   n27 = Get bIsCrouching()
   n28 = "Constraint - DYN_LacesL1 (P)"!(
    ComponentPose=Local To Component(LocalPose=Blend Poses by bool(BlendPose_0=@n25, BlendPose_1=@n19, BlendTime_1=1.5, bActiveValue=@n26) {blendType=Linear})
    bAlphaBoolEnabled:bool=@n27
    ConstraintWeights_0=1
   ) -> (Pose) {
    alphaInputType=Bool
    alphaBoolBlend {blendOption=Linear}
    alphaScaleBiasClamp {scale=0.4}
    boneToModify {boneName=DYN_LacesL1}
    constraintSetup [{targetBone {boneName=calf_r}}]
    constraintWeights [1]
   }
   n29 = "Constraint - DYN_LacesR1 (P)"!(ComponentPose=@n28, bAlphaBoolEnabled:bool=@n27, ConstraintWeights_0=1) -> (Pose) {
    alphaInputType=Bool
    alphaBoolBlend {blendOption=Linear}
    alphaScaleBiasClamp {scale=0.4}
    boneToModify {boneName=DYN_LacesR1}
    constraintSetup [{targetBone {boneName=calf_r}}]
    constraintWeights [1]
   }
   n30 = "Constraint - DYN_LacesL2 (P)"!(ComponentPose=@n29, bAlphaBoolEnabled:bool=@n27, ConstraintWeights_0=1) -> (Pose) {
    alphaInputType=Bool
    alphaBoolBlend {blendOption=Linear}
    alphaScaleBiasClamp {scale=0.4}
    boneToModify {boneName=DYN_LacesL2}
    constraintSetup [{targetBone {boneName=calf_r}}]
    constraintWeights [1]
   }
   n31 = "Constraint - DYN_LacesR2 (P)"!(ComponentPose=@n30, bAlphaBoolEnabled:bool=@n27, ConstraintWeights_0=1) -> (Pose) {
    alphaInputType=Bool
    alphaBoolBlend {blendOption=Linear}
    alphaScaleBiasClamp {scale=0.4}
    boneToModify {boneName=DYN_LacesR2}
    constraintSetup [{targetBone {boneName=calf_r}}]
    constraintWeights [1]
   }
   n32 = "To Float (Boolean)"(InBool=OR Boolean!(A=@n27, B=Get bIsJumping(), C:bool=Get bIsFalling()) -> (ReturnValue))
   n33 = "Transform (Modify) Bone - Bone: DYN_KneeR"(ComponentPose=@n31, Alpha=@n32, Translation="0,0,0", Rotation="0,0,0", Scale="1.000,1.000,0.6") {scaleMode=BMM_Additive, scaleSpace=BCS_BoneSpace, alphaBoolBlend {blendOption=Linear}, boneToModify {boneName=DYN_KneeR}, scale {z=0.6}}
   n34 = "Transform (Modify) Bone - Bone: DYN_KneeL"(ComponentPose=@n33, Alpha=@n32, Translation="0,0,0", Rotation="0,0,0", Scale="1.000,1.000,0.7") {scaleMode=BMM_Additive, scaleSpace=BCS_BoneSpace, alphaBoolBlend {blendOption=Linear}, boneToModify {boneName=DYN_KneeL}, scale {z=0.7}}
   n35 = Get bIsSkyDiving()
   n36 = Get bIsParachuteOpened()
   n37 = Get bIsPlayingEmote()
   n38 = "Constraint - DYN_Armband (PPP)"!(
    ComponentPose=@n34
    Alpha="To Float (Boolean)"(InBool=OR Boolean(A=@n35, B=@n36, C=@n37))
    ConstraintWeights_0=0.33
    ConstraintWeights_1:float=0.33
    ConstraintWeights_2:float=0.34
   ) -> (Pose) {
    alphaBoolBlend {blendOption=Linear}
    boneToModify {boneName=DYN_Armband}
    constraintSetup [{targetBone {boneName=upperarm_twist_01_r}}, {targetBone {boneName=upperarm_twist_02_r}}, {targetBone {boneName=upperarm_r}}]
    constraintWeights [1,1,1]
   }
   n39 = "Transform (Modify) Bone - Bone: Dyn_Necklace_Cntr_03"!(ComponentPose=@n38, bAlphaBoolEnabled:bool=Get bIsInFrontEnd(), Translation="0,-0.3,0", Rotation="0.000,-6,0", Scale="1,1,1") -> (Pose) {
    alphaInputType=Bool
    rotationMode=BMM_Additive
    rotationSpace=BCS_BoneSpace
    translationMode=BMM_Additive
    translationSpace=BCS_BoneSpace
    alphaBoolBlend {blendOption=Linear}
    boneToModify {boneName=Dyn_Necklace_Cntr_03}
    rotation {yaw=-6}
    translation {y=-0.3}
   }
   n40 = "Transform (Modify) Bone - Bone: Dyn_Necklace_Cntr_02"(ComponentPose=@n39, Alpha=1, Translation="0,-0.1,0", Rotation="0,0,0", Scale="1,1,1") {
    alpha=0
    translationMode=BMM_Additive
    translationSpace=BCS_BoneSpace
    alphaBoolBlend {blendOption=Linear}
    boneToModify {boneName=Dyn_Necklace_Cntr_02}
    translation {y=-0.1}
   }
   n41 = "Transform (Modify) Bone - Bone: Dyn_Necklace_Cntr_03"!(ComponentPose=@n40, Alpha:float=1, Translation="0,-0.1,0", Rotation="0,0,0", Scale="1,1,1") -> (Pose) {
    alpha=0
    translationMode=BMM_Additive
    translationSpace=BCS_BoneSpace
    alphaBoolBlend {blendOption=Linear}
    boneToModify {boneName=Dyn_Necklace_Cntr_03}
    translation {y=-0.1}
   }
   n42 = "Transform (Modify) Bone - Bone: Dyn_Necklace_Dogtag_R"(ComponentPose=@n41, Alpha=1, Translation="0,0,0", Rotation="-5,-5,0", Scale="1,1,1") {
    alpha=0
    rotationMode=BMM_Additive
    rotationSpace=BCS_BoneSpace
    alphaBoolBlend {blendOption=Linear}
    boneToModify {boneName=Dyn_Necklace_Dogtag_R}
    rotation {pitch=-5, yaw=-5}
   }
   n43 = "Transform (Modify) Bone - Bone: Dyn_Necklace_Dogtag_L"(ComponentPose=@n42, Alpha=1, Translation="0,0,0", Rotation="0,-5,0", Scale="1,1,1") {
    alpha=0
    rotationMode=BMM_Additive
    rotationSpace=BCS_BoneSpace
    alphaBoolBlend {blendOption=Linear}
    boneToModify {boneName=Dyn_Necklace_Dogtag_L}
    rotation {yaw=-5}
   }
   Output Pose(Result=Component To Local(ComponentPose=@n43)) {layerGroup=None}
  ]}
 EventGraph {nodes [
   Event BlueprintUpdateAnimation()
   > Set isRBANdisabled(isRBANdisabled:bool=NOT Boolean(A=Get bRigidBodyEnabled())) -> (Output_Get:bool)
  ]}
]
variables [isRBANdisabled {type=boolean}]
